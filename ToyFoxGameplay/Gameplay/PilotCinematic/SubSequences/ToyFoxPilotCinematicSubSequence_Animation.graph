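# Animation sub-sequence for the ToyFox pilot cinematic that animates four mannequins belonging to a splash screen mannequin manager.
# Spawn tracks keep the manager Blueprint and a scaled-down root cube present, each with a static transform and an attach track.
# The manager's four child actor components each carry a skeletal animation track, with sections covering the same window partway through the sequence so the mannequins animate simultaneously.
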LevelSequence /ToyFoxGameplay/Gameplay/PilotCinematic/SubSequences/ToyFoxPilotCinematicSubSequence_Animation
bindings [
 MannequinChildActor_02 {
  id=#1
  parent_id=#2
  possessed_class=ChildActorComponent
  tracks [{type=MovieSceneSkeletalAnimationTrack, sections [{type=MovieSceneSkeletalAnimationSection, end_frame=189, start_frame=74}]}]
 }
 MannequinChildActor_04 {
  id=#3
  parent_id=#2
  possessed_class=ChildActorComponent
  tracks [{type=MovieSceneSkeletalAnimationTrack, sections [{type=MovieSceneSkeletalAnimationSection, end_frame=189, start_frame=74}]}]
 }
 SplashScreenMannequinManager {
  id=#2
  possessed_class=B_ToyFoxPilotCinematic_SplashScreenMannequinManager_C
  tracks [
   {type=MovieSceneSpawnTrack, sections [{type=MovieSceneSpawnSection, channels [None {type=MovieSceneScriptingBoolChannel, default=true}]}]}
   {
    type=MovieScene3DTransformTrack
    sections [{
      type=MovieScene3DTransformSection
      channels [
       Location.X {type=MovieSceneScriptingDoubleChannel, default=0}
       Location.Y {type=MovieSceneScriptingDoubleChannel, default=0}
       Location.Z {type=MovieSceneScriptingDoubleChannel, default=0}
       Rotation.X {type=MovieSceneScriptingDoubleChannel, default=0}
       Rotation.Y {type=MovieSceneScriptingDoubleChannel, default=0}
       Rotation.Z {type=MovieSceneScriptingDoubleChannel, default=0}
       Scale.X {type=MovieSceneScriptingDoubleChannel, default=1}
       Scale.Y {type=MovieSceneScriptingDoubleChannel, default=1}
       Scale.Z {type=MovieSceneScriptingDoubleChannel, default=1}
      ]
     }]
   }
   {type=MovieScene3DAttachTrack, sections [{type=MovieScene3DAttachSection, end_frame=654, start_frame=0}]}
  ]
 }
 MannequinChildActor_03 {
  id=#4
  parent_id=#2
  possessed_class=ChildActorComponent
  tracks [{type=MovieSceneSkeletalAnimationTrack, sections [{type=MovieSceneSkeletalAnimationSection, end_frame=189, start_frame=74}]}]
 }
 MannequinRootCube {
  id=#5
  possessed_class=FortStaticMeshActor
  tracks [
   {
    type=MovieScene3DTransformTrack
    sections [{
      type=MovieScene3DTransformSection
      channels [
       Location.X {type=MovieSceneScriptingDoubleChannel, default=500}
       Location.Y {type=MovieSceneScriptingDoubleChannel, default=0}
       Location.Z {type=MovieSceneScriptingDoubleChannel, default=0}
       Rotation.X {type=MovieSceneScriptingDoubleChannel, default=0}
       Rotation.Y {type=MovieSceneScriptingDoubleChannel, default=0}
       Rotation.Z {type=MovieSceneScriptingDoubleChannel, default=0}
       Scale.X {type=MovieSceneScriptingDoubleChannel, default=0.25}
       Scale.Y {type=MovieSceneScriptingDoubleChannel, default=0.25}
       Scale.Z {type=MovieSceneScriptingDoubleChannel, default=0.25}
      ]
     }]
   }
   {type=MovieSceneSpawnTrack, sections [{type=MovieSceneSpawnSection, channels [None {type=MovieSceneScriptingBoolChannel, default=true}]}]}
   {type=MovieScene3DAttachTrack, sections [{type=MovieScene3DAttachSection, end_frame=654, start_frame=0}]}
  ]
 }
 MannequinChildActor_01 {
  id=#6
  parent_id=#2
  possessed_class=ChildActorComponent
  tracks [{type=MovieSceneSkeletalAnimationTrack, sections [{type=MovieSceneSkeletalAnimationSection, end_frame=189, start_frame=74}]}]
 }
]
display_rate {denominator=1, numerator=30}
tick_resolution {denominator=1, numerator=24000}
tracks []
