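# Editor-only Level Sequence, named as a sub-sequence of the ToyFox pilot cinematic, that stages the giant player pawn.
# The pawn has a spawn track, a visibility track and an unkeyed transform that holds it at a fixed world position.
# Its child CharacterMesh0 plays two back-to-back skeletal animation sections, with weight keys on the first that drop to zero.
# The sequence has no root tracks, such as camera cuts or audio.
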LevelSequence /ToyFoxGameplay/Gameplay/PilotCinematic/SubSequences/ToyFoxPilotCinematicSubSequence_EditorOnly
bindings [
 CharacterMesh0 {
  id=#1
  parent_id=#2
  possessed_class=FortPooledSkelMeshComponent
  tracks [{
    type=MovieSceneSkeletalAnimationTrack
    sections [
     {
      type=MovieSceneSkeletalAnimationSection
      end_frame=293
      start_frame=0
      channels [Weight {type=MovieSceneScriptingFloatChannel, default=1, keys [[12,1],[305,1],[338,0]]}]
     }
     {type=MovieSceneSkeletalAnimationSection, end_frame=637, start_frame=293}
    ]
   }]
 }
 B_ToyFoxGiantPlayerPawn {
  id=#2
  possessed_class=B_ToyFoxGiantPlayerPawn_C
  tracks [
   {type=MovieSceneSpawnTrack, sections [{type=MovieSceneSpawnSection, channels [None {type=MovieSceneScriptingBoolChannel, default=true}]}]}
   {
    type=MovieScene3DTransformTrack
    sections [{
      type=MovieScene3DTransformSection
      channels [
       Location.X {type=MovieSceneScriptingDoubleChannel, default=29360}
       Location.Y {type=MovieSceneScriptingDoubleChannel, default=-25048}
       Location.Z {type=MovieSceneScriptingDoubleChannel, default=2048}
       Rotation.X {type=MovieSceneScriptingDoubleChannel, default=0}
       Rotation.Y {type=MovieSceneScriptingDoubleChannel, default=0}
       Rotation.Z {type=MovieSceneScriptingDoubleChannel, default=0}
       Scale.X {type=MovieSceneScriptingDoubleChannel, default=1}
       Scale.Y {type=MovieSceneScriptingDoubleChannel, default=1}
       Scale.Z {type=MovieSceneScriptingDoubleChannel, default=1}
      ]
     }]
   }
   {
    type=MovieSceneVisibilityTrack
    sections [{type=MovieSceneVisibilitySection, channels [None {type=MovieSceneScriptingBoolChannel, default=true}]}]
   }
  ]
 }
]
display_rate {denominator=1, numerator=30}
tick_resolution {denominator=1, numerator=24000}
tracks []
