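# Gameplay sub-sequence of the ToyFox pilot cinematic, a Level Sequence that carries only timed event triggers.
# A single unbound event track with one trigger section fires five events at set points on a 30 fps timeline.
# It has no bindings or other tracks, and the events show no details of what they call.
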LevelSequence /ToyFoxGameplay/Gameplay/PilotCinematic/SubSequences/ToyFoxPilotCinematicSubSequence_Gameplay
bindings []
display_rate {denominator=1, numerator=30}
tick_resolution {denominator=1, numerator=24000}
tracks [{
  type=MovieSceneEventTrack
  sections [{
    type=MovieSceneEventTriggerSection
    channels [None {
      type=MovieSceneScriptingEventChannel
      keys [
       [168, "<Struct 'MovieSceneEvent' (0x85369765820) {}>"]
       [259, "<Struct 'MovieSceneEvent' (0x85369765aa0) {}>"]
       [263, "<Struct 'MovieSceneEvent' (0x85369765780) {}>"]
       [300, "<Struct 'MovieSceneEvent' (0x85369765640) {}>"]
       [412, "<Struct 'MovieSceneEvent' (0x85369765500) {}>"]
      ]
     }]
   }]
 }]
