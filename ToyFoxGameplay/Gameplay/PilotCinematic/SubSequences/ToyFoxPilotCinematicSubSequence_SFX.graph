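# Level Sequence that serves as the sound-effects subsequence of the ToyFox pilot cinematic.
# It holds a single audio track with one section that starts partway into the 30 fps timeline.
# Volume and pitch stay at their defaults, and there are no object bindings.
# No sound asset is shown on the section.
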LevelSequence /ToyFoxGameplay/Gameplay/PilotCinematic/SubSequences/ToyFoxPilotCinematicSubSequence_SFX
bindings []
display_rate {denominator=1, numerator=30}
tick_resolution {denominator=1, numerator=24000}
tracks [{
  type=MovieSceneAudioTrack
  display_name=Audio
  sections [{
    type=MovieSceneAudioSection
    end_frame=407
    start_frame=75
    channels [
     Volume {type=MovieSceneScriptingFloatChannel, default=1}
     Pitch {type=MovieSceneScriptingFloatChannel, default=1}
     AttachActor {type=MovieSceneScriptingActorReferenceChannel, default="<Struct 'MovieSceneObjectBindingID' (0x853696910c0) {}>"}
    ]
   }]
 }]
