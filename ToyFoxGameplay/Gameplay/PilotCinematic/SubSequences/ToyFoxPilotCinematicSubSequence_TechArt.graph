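# TechArt sub-sequence of the ToyFox pilot cinematic that manages a single background plane mesh.
# A spawn track keeps the static mesh actor spawned, and an attach track parents it to a target not shown here for the length of the shot.
# A transform track with constant values holds the plane scaled up, rotated and offset along X, with no animation.
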LevelSequence /ToyFoxGameplay/Gameplay/PilotCinematic/SubSequences/ToyFoxPilotCinematicSubSequence_TechArt
bindings [BackgroundPlaneMesh {
  id=#1
  possessed_class=FortStaticMeshActor
  tracks [
   {
    type=MovieScene3DTransformTrack
    sections [{
      type=MovieScene3DTransformSection
      channels [
       Location.X {type=MovieSceneScriptingDoubleChannel, default=1000}
       Location.Y {type=MovieSceneScriptingDoubleChannel, default=0}
       Location.Z {type=MovieSceneScriptingDoubleChannel, default=0}
       Rotation.X {type=MovieSceneScriptingDoubleChannel, default=-90}
       Rotation.Y {type=MovieSceneScriptingDoubleChannel, default=0}
       Rotation.Z {type=MovieSceneScriptingDoubleChannel, default=-90}
       Scale.X {type=MovieSceneScriptingDoubleChannel, default=8}
       Scale.Y {type=MovieSceneScriptingDoubleChannel, default=8}
       Scale.Z {type=MovieSceneScriptingDoubleChannel, default=8}
      ]
     }]
   }
   {type=MovieSceneSpawnTrack, sections [{type=MovieSceneSpawnSection, channels [None {type=MovieSceneScriptingBoolChannel, default=true}]}]}
   {type=MovieScene3DAttachTrack, sections [{type=MovieScene3DAttachSection, end_frame=610, start_frame=-5}]}
  ]
 }]
display_rate {denominator=1, numerator=30}
tick_resolution {denominator=1, numerator=24000}
tracks []
