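# Level Sequence named as the VFX sub-sequence of the ToyFox pilot cinematic.
# It has no bindings or tracks, so it animates and spawns nothing as it stands.
# Only its timing is shown, a 30 fps display rate over a 24000-per-second tick resolution.
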LevelSequence /ToyFoxGameplay/Gameplay/PilotCinematic/SubSequences/ToyFoxPilotCinematicSubSequence_VFX
bindings []
display_rate {denominator=1, numerator=30}
tick_resolution {denominator=1, numerator=24000}
tracks []
