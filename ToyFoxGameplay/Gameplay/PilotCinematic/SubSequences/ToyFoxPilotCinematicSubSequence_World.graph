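# World subsequence of the ToyFox pilot cinematic, controlling the cockpit set, its mannequins and the time of day.
# Spawn tracks bring in the cockpit actor partway through and remove the time-of-day actor later on.
# Five mannequin child-actor components under the cockpit play short, slightly staggered skeletal animation sections.
# The attached time-of-day actor's DaySequenceModifier component holds its Time at early evening.
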LevelSequence /ToyFoxGameplay/Gameplay/PilotCinematic/SubSequences/ToyFoxPilotCinematicSubSequence_World
bindings [
 MannequinChildActor03 {
  id=#1
  parent_id=#2
  possessed_class=ChildActorComponent
  tracks [{type=MovieSceneSkeletalAnimationTrack, sections [{type=MovieSceneSkeletalAnimationSection, end_frame=246, start_frame=166}]}]
 }
 MannequinChildActor05 {
  id=#3
  parent_id=#2
  possessed_class=ChildActorComponent
  tracks [{type=MovieSceneSkeletalAnimationTrack, sections [{type=MovieSceneSkeletalAnimationSection, end_frame=246, start_frame=166}]}]
 }
 B_ToyFoxPilotCinematic_Cockpit {
  id=#2
  possessed_class=B_ToyFoxPilotCinematic_Cockpit_C
  tracks [{
    type=MovieSceneSpawnTrack
    sections [{type=MovieSceneSpawnSection, channels [None {type=MovieSceneScriptingBoolChannel, default=false, keys [[99,false],[120,true]]}]}]
   }]
 }
 MannequinChildActor01 {
  id=#4
  parent_id=#2
  possessed_class=ChildActorComponent
  tracks [{type=MovieSceneSkeletalAnimationTrack, sections [{type=MovieSceneSkeletalAnimationSection, end_frame=248, start_frame=168}]}]
 }
 B_ToyFoxPilotCinematic_TOD {
  id=#5
  possessed_class=B_ToyFoxPilotCinematic_TOD_C
  tracks [
   {
    type=MovieSceneSpawnTrack
    sections [{type=MovieSceneSpawnSection, channels [None {type=MovieSceneScriptingBoolChannel, default=false, keys [[173,true],[174,false]]}]}]
   }
   {
    type=MovieScene3DTransformTrack
    sections [{
      type=MovieScene3DTransformSection
      channels [
       Location.X {type=MovieSceneScriptingDoubleChannel, default=0}
       Location.Y {type=MovieSceneScriptingDoubleChannel, default=0}
       Location.Z {type=MovieSceneScriptingDoubleChannel, default=0}
       Rotation.X {type=MovieSceneScriptingDoubleChannel, default=0}
       Rotation.Y {type=MovieSceneScriptingDoubleChannel, default=0}
       Rotation.Z {type=MovieSceneScriptingDoubleChannel, default=0}
       Scale.X {type=MovieSceneScriptingDoubleChannel, default=1}
       Scale.Y {type=MovieSceneScriptingDoubleChannel, default=1}
       Scale.Z {type=MovieSceneScriptingDoubleChannel, default=1}
      ]
     }]
   }
   {type=MovieScene3DAttachTrack, sections [{type=MovieScene3DAttachSection, end_frame=600, start_frame=0}]}
  ]
 }
 MannequinChildActor04 {
  id=#6
  parent_id=#2
  possessed_class=ChildActorComponent
  tracks [{type=MovieSceneSkeletalAnimationTrack, sections [{type=MovieSceneSkeletalAnimationSection, end_frame=248, start_frame=168}]}]
 }
 DaySequenceModifier {
  id=#7
  parent_id=#5
  possessed_class=DaySequenceModifierComponent
  tracks [{
    type=MovieSceneFloatTrack
    display_name=Time
    sections [{type=MovieSceneFloatSection, channels [None {type=MovieSceneScriptingFloatChannel, default=18.5, const [18.5,2]}]}]
   }]
 }
 MannequinChildActor02 {
  id=#8
  parent_id=#2
  possessed_class=ChildActorComponent
  tracks [{type=MovieSceneSkeletalAnimationTrack, sections [{type=MovieSceneSkeletalAnimationSection, end_frame=244, start_frame=164}]}]
 }
]
display_rate {denominator=1, numerator=30}
tick_resolution {denominator=1, numerator=24000}
tracks []
